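# Sound cue for the kayak close sound of the Lobster item, classed as weapon foley.
# A single wave player feeds the output directly, with no randomization or modulation.
# It borrows the Lobster melee equip attenuation, plays at reduced volume and overrides concurrency to cap simultaneous instances, stopping the lowest priority one.
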SoundCue /Galileo/Items/Lobster/Assets/Sounds/Kayak/Lobster_Kayak_Close_Cue
root=WavePlayer_0
nodes [{id=WavePlayer_0, properties {soundWaveAssetPtr=&/Lobster_Kayak_Close_01}}]
properties {
 attenuationSettings=/Game/Sounds/Attenuation/Fort_Weapon_Melee_Lobster_Equip_Att
 bOverrideConcurrency=true
 maxDistance=4000
 soundClassObject=&Fort_WeaponFoley
 volumeMultiplier=0.113
 concurrencyOverrides {maxCount=2, resolutionRule=StopLowestPriority}
}
