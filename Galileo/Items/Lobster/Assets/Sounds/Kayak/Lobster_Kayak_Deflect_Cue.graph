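# Sound cue for deflects with the Kayak Lobster melee weapon, using melee attenuation and a capped concurrency override.
# A mixer layers random deflect hits, switched by quality level between ten-wave and four-wave pools, with a random poise warning sound.
# The InvPoise parameter modulates pitch on both layers and crossfades the warning in only near the top of its 0 to 1 range.
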SoundCue /Galileo/Items/Lobster/Assets/Sounds/Kayak/Lobster_Kayak_Deflect_Cue
root=Mixer_0
nodes [
 {id=Mixer_0, children [ModulatorContinuous_1, ParamCrossFade_0], properties {inputVolume [1,1]}}
 {
  id=ModulatorContinuous_1
  children [QualityLevel_0]
  properties {pitchModulationParams {default=0, maxOutput=1.2, minOutput=0.8, parameterName=InvPoise}}
 }
 {
  id=ModulatorContinuous_2
  children [Random_0]
  properties {pitchModulationParams {default=0, maxOutput=1.5, minOutput=0.9, parameterName=InvPoise}}
 }
 {
  id=ParamCrossFade_0
  children [ModulatorContinuous_2]
  properties {
   paramName=InvPoise
   crossFadeInput [{fadeInDistanceEnd=1, fadeInDistanceStart=0.8, fadeOutDistanceEnd=100, fadeOutDistanceStart=1, volume=1.2}]
  }
 }
 {id=QualityLevel_0, children [Random_1,Random_1,Random_2,Random_2]}
 {id=Random_0, children [WavePlayer_1,WavePlayer_2], properties {weights [1,1]}}
 {
  id=Random_1
  children [
   WavePlayer_0
   WavePlayer_7
   WavePlayer_8
   WavePlayer_9
   WavePlayer_10
   WavePlayer_11
   WavePlayer_12
   WavePlayer_13
   WavePlayer_14
   WavePlayer_15
  ]
  properties {weights [1,1,1,1,1,1,1,1,1,1]}
 }
 {id=Random_2, children [WavePlayer_3,WavePlayer_4,WavePlayer_5,WavePlayer_6], properties {weights [1,1,1,1]}}
 {id=WavePlayer_0, properties {soundWaveAssetPtr=&/Lobster_Kayak_Deflect_10}}
 {id=WavePlayer_1, properties {soundWaveAssetPtr=&/Lobster_Kayak_Poise_Warning_01}}
 {id=WavePlayer_10, properties {soundWaveAssetPtr=&/Lobster_Kayak_Deflect_04}}
 {id=WavePlayer_11, properties {soundWaveAssetPtr=&/Lobster_Kayak_Deflect_05}}
 {id=WavePlayer_12, properties {soundWaveAssetPtr=&/Lobster_Kayak_Deflect_06}}
 {id=WavePlayer_13, properties {soundWaveAssetPtr=&/Lobster_Kayak_Deflect_07}}
 {id=WavePlayer_14, properties {soundWaveAssetPtr=&/Lobster_Kayak_Deflect_08}}
 {id=WavePlayer_15, properties {soundWaveAssetPtr=&/Lobster_Kayak_Deflect_09}}
 {id=WavePlayer_2, properties {soundWaveAssetPtr=&/Lobster_Kayak_Poise_Warning_02}}
 {id=WavePlayer_3, properties {soundWaveAssetPtr=&Lobster_Kayak_Deflect_06}}
 {id=WavePlayer_4, properties {soundWaveAssetPtr=&Lobster_Kayak_Deflect_07}}
 {id=WavePlayer_5, properties {soundWaveAssetPtr=&Lobster_Kayak_Deflect_08}}
 {id=WavePlayer_6, properties {soundWaveAssetPtr=&Lobster_Kayak_Deflect_09}}
 {id=WavePlayer_7, properties {soundWaveAssetPtr=&/Lobster_Kayak_Deflect_01}}
 {id=WavePlayer_8, properties {soundWaveAssetPtr=&/Lobster_Kayak_Deflect_02}}
 {id=WavePlayer_9, properties {soundWaveAssetPtr=&/Lobster_Kayak_Deflect_03}}
]
properties {
 attenuationSettings=/Game/Sounds/Attenuation/Fort_Weapon_Melee_Lobster_Attack_Deflect_Stereo_Att
 bOverrideConcurrency=true
 maxDistance=12000
 soundClassObject=&Fort_WeaponMelee
 volumeMultiplier=0.6748
 concurrencyOverrides {maxCount=3, resolutionRule=StopLowestPriority}
}
