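# Sound cue for the dodge sound of the Lobster Kayak item, classed as melee weapon audio.
# A random node picks one of three equally weighted dodge wave variations.
# It uses a close-range stereo melee attack attenuation and a slightly lowered volume.
# A concurrency override caps it at two simultaneous plays, stopping the lowest priority instance.
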SoundCue /Galileo/Items/Lobster/Assets/Sounds/Kayak/Lobster_Kayak_Dodge_Cue
root=Random_0
nodes [
 {id=Random_0, children [WavePlayer_1,WavePlayer_2,WavePlayer_3], properties {weights [1,1,1]}}
 {id=WavePlayer_1, properties {soundWaveAssetPtr=&/Lobster_Kayak_Dodge_01}}
 {id=WavePlayer_2, properties {soundWaveAssetPtr=&/Lobster_Kayak_Dodge_02}}
 {id=WavePlayer_3, properties {soundWaveAssetPtr=&/Lobster_Kayak_Dodge_03}}
]
properties {
 attenuationSettings=/Game/Sounds/Attenuation/Fort_Weapon_Melee_Attack_Close_Range_Stereo_Att
 bOverrideConcurrency=true
 maxDistance=10058.75
 soundClassObject=&Fort_WeaponMelee
 volumeMultiplier=0.731
 concurrencyOverrides {maxCount=2, resolutionRule=StopLowestPriority}
}
