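# Sound cue for dropping the Kayak variant of the Lobster item, reusing the Default variant's drop impact sound.
# A single wave player feeds the root directly, with no randomization or mixing, at reduced volume.
# It uses weapon-drop attenuation and the loot-collect sound class, and references a shared stop-oldest weapon-drop concurrency set.
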SoundCue /Galileo/Items/Lobster/Assets/Sounds/Kayak/Lobster_Kayak_Drop_Cue
root=WavePlayer_0
nodes [{id=WavePlayer_0, properties {soundWaveAssetPtr=/Galileo/Items/Lobster/Assets/Sounds/Default/Lobster_Default_DropImpact_01}}]
properties {
 attenuationSettings=&Fort_Weapon_Drop_Att
 maxDistance=3000
 soundClassObject=&Fort_Loot_Collect
 volumeMultiplier=0.5623
 concurrencyOverrides {resolutionRule=StopLowestPriority}
 concurrencySet [&Weapon_Drop_StopOldest]
}
