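# Sound cue for the looping idle sound of the Lobster Kayak melee weapon.
# A single looping wave player feeds the root directly, with raised pitch and reduced volume applied at the cue level.
# It uses the melee weapon sound class and a close-range idle-loop attenuation asset.
# A concurrency override limits simultaneous instances, stopping the lowest priority when the limit is exceeded.
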SoundCue /Galileo/Items/Lobster/Assets/Sounds/Kayak/Lobster_Kayak_IdleLoop_Cue
root=WavePlayer_0
nodes [{id=WavePlayer_0, properties {bLooping=true, soundWaveAssetPtr=&/Lobster_Kayak_IdleLoop_01}}]
properties {
 attenuationSettings=/Game/Sounds/Attenuation/Fort_Weapon_Melee_Lobster_Idle_Loop_Close_Att
 bOverrideConcurrency=true
 maxDistance=3000
 pitchMultiplier=1.25
 soundClassObject=&Fort_WeaponMelee
 volumeMultiplier=0.5623
 concurrencyOverrides {maxCount=3, resolutionRule=StopLowestPriority}
}
