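# Sound cue for kayak impact sounds on the Lobster item, classed as a melee weapon sound.
# A Random node picks one of six equally weighted impact waves, excluded from branch culling.
# Playback is quieter and pitched up, uses close-range stereo melee attenuation, and overrides concurrency to a few simultaneous voices, stopping the lowest priority.
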SoundCue /Galileo/Items/Lobster/Assets/Sounds/Kayak/Lobster_Kayak_Impact_Cue
root=Random_0
nodes [
 {
  id=Random_0
  children [WavePlayer_1,WavePlayer_2,WavePlayer_3,WavePlayer_4,WavePlayer_5,WavePlayer_6]
  properties {bShouldExcludeFromBranchCulling=true, weights [1,1,1,1,1,1]}
 }
 {id=WavePlayer_1, properties {soundWaveAssetPtr=&/Lobster_Kayak_Impact_01}}
 {id=WavePlayer_2, properties {soundWaveAssetPtr=&/Lobster_Kayak_Impact_02}}
 {id=WavePlayer_3, properties {soundWaveAssetPtr=&/Lobster_Kayak_Impact_03}}
 {id=WavePlayer_4, properties {soundWaveAssetPtr=&/Lobster_Kayak_Impact_04}}
 {id=WavePlayer_5, properties {soundWaveAssetPtr=&/Lobster_Kayak_Impact_05}}
 {id=WavePlayer_6, properties {soundWaveAssetPtr=&/Lobster_Kayak_Impact_06}}
]
properties {
 attenuationSettings=/Game/Sounds/Attenuation/Fort_Weapon_Melee_Attack_Close_Range_Stereo_Att
 bOverrideConcurrency=true
 maxDistance=10058.75
 pitchMultiplier=1.2
 soundClassObject=&Fort_WeaponMelee
 volumeMultiplier=0.3
 concurrencyOverrides {maxCount=3, resolutionRule=StopLowestPriority}
}
