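# Sound cue for the Lobster Kayak melee weapon's opening, combining a one-shot open sound with a looping idle.
# After a short initial delay, a mixer layers one of four equally weighted random open waves, boosted in volume, with the idle loop, which starts after a second short delay.
# The cue uses the Lobster equip attenuation and melee weapon sound class, with a concurrency override that stops lowest-priority instances.
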SoundCue /Galileo/Items/Lobster/Assets/Sounds/Kayak/Lobster_Kayak_Open_Cue
root=Delay_0
nodes [
 {id=Delay_0, children [Mixer_0], properties {delayMax=0.2, delayMin=0.2}}
 {id=Delay_1, children [WavePlayer_0], properties {delayMax=0.2, delayMin=0.2}}
 {id=Mixer_0, children [Random_0,Delay_1], properties {inputVolume [1.5,1]}}
 {id=Random_0, children [WavePlayer_1,WavePlayer_2,WavePlayer_3,WavePlayer_4], properties {weights [1,1,1,1]}}
 {id=WavePlayer_0, properties {bLooping=true, soundWaveAssetPtr=&/Lobster_Kayak_IdleLoop_01}}
 {id=WavePlayer_1, properties {soundWaveAssetPtr=&/Lobster_Kayak_Open_01}}
 {id=WavePlayer_2, properties {soundWaveAssetPtr=&/Lobster_Kayak_Open_02}}
 {id=WavePlayer_3, properties {soundWaveAssetPtr=&/Lobster_Kayak_Open_03}}
 {id=WavePlayer_4, properties {soundWaveAssetPtr=&/Lobster_Kayak_Open_04}}
]
properties {
 attenuationSettings=/Game/Sounds/Attenuation/Fort_Weapon_Melee_Lobster_Equip_Att
 bOverrideConcurrency=true
 maxDistance=4000
 soundClassObject=&Fort_WeaponMelee
 volumeMultiplier=0.4218
 concurrencyOverrides {maxCount=3, resolutionRule=StopLowestPriority}
}
