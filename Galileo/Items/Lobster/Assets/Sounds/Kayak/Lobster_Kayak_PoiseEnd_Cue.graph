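# Sound cue for the end of poise on the Lobster Kayak item, under Galileo items.
# A single wave player feeds the root directly with the PoiseEnd_01 wave, with no randomization or modulation.
# It uses the melee weapon sound class and a close-range stereo melee attack attenuation at reduced volume.
# A concurrency override caps simultaneous instances, stopping the lowest priority one.
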SoundCue /Galileo/Items/Lobster/Assets/Sounds/Kayak/Lobster_Kayak_PoiseEnd_Cue
root=WavePlayer_1
nodes [{id=WavePlayer_1, properties {soundWaveAssetPtr=&/Lobster_Kayak_PoiseEnd_01}}]
properties {
 attenuationSettings=/Game/Sounds/Attenuation/Fort_Weapon_Melee_Attack_Close_Range_Stereo_Att
 bOverrideConcurrency=true
 maxDistance=10058.75
 soundClassObject=&Fort_WeaponMelee
 volumeMultiplier=0.4218
 concurrencyOverrides {maxCount=2, resolutionRule=StopLowestPriority}
}
