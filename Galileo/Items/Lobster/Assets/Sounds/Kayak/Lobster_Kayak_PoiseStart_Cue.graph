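# Sound cue for the start of the Lobster Kayak melee item's poise sound.
# A random node picks one of three equally weighted PoiseStart wave variations.
# It plays at reduced volume in the melee weapon sound class with close-range stereo melee attenuation.
# A concurrency override limits simultaneous instances to two, stopping the lowest priority one.
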SoundCue /Galileo/Items/Lobster/Assets/Sounds/Kayak/Lobster_Kayak_PoiseStart_Cue
root=Random_1
nodes [
 {id=Random_1, children [WavePlayer_0, WavePlayer_9, WavePlayer_10], properties {weights [1,1,1]}}
 {id=WavePlayer_0, properties {soundWaveAssetPtr=&/Lobster_Kayak_PoiseStart_01}}
 {id=WavePlayer_10, properties {soundWaveAssetPtr=&/Lobster_Kayak_PoiseStart_03}}
 {id=WavePlayer_9, properties {soundWaveAssetPtr=&/Lobster_Kayak_PoiseStart_02}}
]
properties {
 attenuationSettings=/Game/Sounds/Attenuation/Fort_Weapon_Melee_Attack_Close_Range_Stereo_Att
 bOverrideConcurrency=true
 maxDistance=10058.75
 soundClassObject=&Fort_WeaponMelee
 volumeMultiplier=0.5623
 concurrencyOverrides {maxCount=2, resolutionRule=StopLowestPriority}
}
