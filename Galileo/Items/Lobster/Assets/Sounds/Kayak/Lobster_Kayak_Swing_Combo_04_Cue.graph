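# Sound cue for the fourth swing in the Lobster Kayak melee combo, played as a close-range weapon melee sound.
# All quality levels route to one Random node that picks evenly among four SwingFour wave variations and is excluded from branch culling.
# It uses stereo close-range melee attenuation and the melee weapon sound class, with owner and global swing concurrency sets limiting simultaneous plays.
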SoundCue /Galileo/Items/Lobster/Assets/Sounds/Kayak/Lobster_Kayak_Swing_Combo_04_Cue
root=QualityLevel_0
nodes [
 {id=QualityLevel_0, children [Random_2,Random_2,Random_2,Random_2]}
 {
  id=Random_2
  children [WavePlayer_12, WavePlayer_13, WavePlayer_14, WavePlayer_15]
  properties {bShouldExcludeFromBranchCulling=true, weights [1,1,1,1]}
 }
 {id=WavePlayer_12, properties {soundWaveAssetPtr=&/Lobster_Kayak_SwingFour_04}}
 {id=WavePlayer_13, properties {soundWaveAssetPtr=&/Lobster_Kayak_SwingFour_01}}
 {id=WavePlayer_14, properties {soundWaveAssetPtr=&/Lobster_Kayak_SwingFour_02}}
 {id=WavePlayer_15, properties {soundWaveAssetPtr=&/Lobster_Kayak_SwingFour_03}}
]
properties {
 attenuationSettings=/Game/Sounds/Attenuation/Fort_Weapon_Melee_Attack_Close_Range_Stereo_Att
 maxDistance=10058.75
 soundClassObject=&Fort_WeaponMelee
 volumeMultiplier=0.731
 concurrencyOverrides {resolutionRule=StopLowestPriority}
 concurrencySet [/Galileo/Items/Lobster/Assets/Sounds/Lobster_Swings_OwnerLimit_Conc, /Galileo/Items/Lobster/Assets/Sounds/Lobster_Swings_GlobalLimit_Conc]
}
