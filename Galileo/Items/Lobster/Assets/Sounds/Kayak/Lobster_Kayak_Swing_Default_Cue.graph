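# Sound cue for the default swing of the Lobster Kayak item, classed as a melee weapon sound.
# A quality-level node sends the first two levels to a random pick of eight swing waves and the last two to a pick from the first four.
# Close-range stereo melee attenuation shapes falloff, while per-owner and global concurrency sets limit overlapping swings.
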SoundCue /Galileo/Items/Lobster/Assets/Sounds/Kayak/Lobster_Kayak_Swing_Default_Cue
root=QualityLevel_0
nodes [
 {id=QualityLevel_0, children [Random_0,Random_0,Random_1,Random_1]}
 {
  id=Random_0
  children [WavePlayer_1,WavePlayer_2,WavePlayer_3,WavePlayer_4,WavePlayer_5,WavePlayer_6,WavePlayer_7,WavePlayer_8]
  properties {weights [1,1,1,1,1,1,1,1]}
 }
 {
  id=Random_1
  children [WavePlayer_0, WavePlayer_9, WavePlayer_10, WavePlayer_11]
  properties {bShouldExcludeFromBranchCulling=true, weights [1,1,1,1]}
 }
 {id=WavePlayer_0, properties {soundWaveAssetPtr=&/Lobster_Kayak_Swing_01}}
 {id=WavePlayer_1, properties {soundWaveAssetPtr=&Lobster_Kayak_Swing_01}}
 {id=WavePlayer_10, properties {soundWaveAssetPtr=&/Lobster_Kayak_Swing_03}}
 {id=WavePlayer_11, properties {soundWaveAssetPtr=&/Lobster_Kayak_Swing_04}}
 {id=WavePlayer_2, properties {soundWaveAssetPtr=&/Lobster_Kayak_Swing_02}}
 {id=WavePlayer_3, properties {soundWaveAssetPtr=&Lobster_Kayak_Swing_03}}
 {id=WavePlayer_4, properties {soundWaveAssetPtr=&Lobster_Kayak_Swing_04}}
 {id=WavePlayer_5, properties {soundWaveAssetPtr=&/Lobster_Kayak_Swing_05}}
 {id=WavePlayer_6, properties {soundWaveAssetPtr=&/Lobster_Kayak_Swing_06}}
 {id=WavePlayer_7, properties {soundWaveAssetPtr=&/Lobster_Kayak_Swing_07}}
 {id=WavePlayer_8, properties {soundWaveAssetPtr=&/Lobster_Kayak_Swing_08}}
 {id=WavePlayer_9, properties {soundWaveAssetPtr=&Lobster_Kayak_Swing_02}}
]
properties {
 attenuationSettings=/Game/Sounds/Attenuation/Fort_Weapon_Melee_Attack_Close_Range_Stereo_Att
 maxDistance=10058.75
 soundClassObject=&Fort_WeaponMelee
 volumeMultiplier=0.5623
 concurrencyOverrides {resolutionRule=StopLowestPriority}
 concurrencySet [/Galileo/Items/Lobster/Assets/Sounds/Lobster_Swings_OwnerLimit_Conc, /Galileo/Items/Lobster/Assets/Sounds/Lobster_Swings_GlobalLimit_Conc]
}
