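# Sound cue for the light swing of the Lobster Kayak item, classed as a melee weapon sound.
# A Random node picks with equal weight among eight light-swing wave variations.
# Playback uses close-range stereo melee attenuation at a reduced volume.
# Shared Lobster swing concurrency sets apply per-owner and global limits to overlapping swings.
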SoundCue /Galileo/Items/Lobster/Assets/Sounds/Kayak/Lobster_Kayak_Swing_Light_Cue
root=Random_1
nodes [
 {
  id=Random_1
  children [WavePlayer_0, WavePlayer_9, WavePlayer_10, WavePlayer_11, WavePlayer_12, WavePlayer_13, WavePlayer_14, WavePlayer_15]
  properties {weights [1,1,1,1,1,1,1,1]}
 }
 {id=WavePlayer_0, properties {soundWaveAssetPtr=&/Lobster_Kayak_SwingLight_01}}
 {id=WavePlayer_10, properties {soundWaveAssetPtr=&/Lobster_Kayak_SwingLight_03}}
 {id=WavePlayer_11, properties {soundWaveAssetPtr=&/Lobster_Kayak_SwingLight_04}}
 {id=WavePlayer_12, properties {soundWaveAssetPtr=&/Lobster_Kayak_SwingLight_05}}
 {id=WavePlayer_13, properties {soundWaveAssetPtr=&/Lobster_Kayak_SwingLight_06}}
 {id=WavePlayer_14, properties {soundWaveAssetPtr=&/Lobster_Kayak_SwingLight_07}}
 {id=WavePlayer_15, properties {soundWaveAssetPtr=&/Lobster_Kayak_SwingLight_08}}
 {id=WavePlayer_9, properties {soundWaveAssetPtr=&/Lobster_Kayak_SwingLight_02}}
]
properties {
 attenuationSettings=/Game/Sounds/Attenuation/Fort_Weapon_Melee_Attack_Close_Range_Stereo_Att
 maxDistance=10058.75
 priority=0.5
 soundClassObject=&Fort_WeaponMelee
 volumeMultiplier=0.1406
 concurrencyOverrides {resolutionRule=StopLowestPriority}
 concurrencySet [/Galileo/Items/Lobster/Assets/Sounds/Lobster_Swings_OwnerLimit_Conc, /Galileo/Items/Lobster/Assets/Sounds/Lobster_Swings_GlobalLimit_Conc]
}
